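# Looping ambient spot cue for the Helios Tower POI that plays randomized mud-blob bubble bursts as a positional sound.
# After each random delay, a mixer plays a full-set random blob alongside a quieter, further-delayed layer from the same five waves, each with pitch and volume modulation.
# The quieter layer's random node also has three empty, equally weighted slots.
# Overrides add focus-based, filtered attenuation and limited concurrency, with an exterior ambient bus send.
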SoundCue /Game/Sounds/Helios/POI/Tower/Spots/Spot_Bubble_Explosion_Cue
root=Looping_0
nodes [
 {id=Delay_0, children [Mixer_0], properties {delayMax=0.6, delayMin=0.3}}
 {id=Delay_1, children [Modulator_1], properties {delayMax=0.5, delayMin=0.2}}
 {id=Looping_0, children [Delay_0]}
 {id=Mixer_0, children [Modulator_0,Delay_1], properties {inputVolume [0.8,0.25]}}
 {id=Modulator_0, children [Random_1], properties {pitchMax=1.2, pitchMin=0.9, volumeMax=1.2, volumeMin=1}}
 {id=Modulator_1, children [Random_0], properties {pitchMax=1.2, pitchMin=0.8, volumeMax=1, volumeMin=0.8}}
 {
  id=Random_0
  children [WavePlayer_0,WavePlayer_1,WavePlayer_2,WavePlayer_3,WavePlayer_4,null,null,null]
  properties {weights [1,1,1,1,1,1,1,1]}
 }
 {id=Random_1, children [WavePlayer_5,WavePlayer_6,WavePlayer_7,WavePlayer_8,WavePlayer_9], properties {weights [1,1,1,1,1]}}
 {id=WavePlayer_0, properties {soundWaveAssetPtr=/Game/Sounds/Fort_Swamp_Sounds/MudBlobB002}}
 {id=WavePlayer_1, properties {soundWaveAssetPtr=/Game/Sounds/Fort_Swamp_Sounds/MudBlobB004}}
 {id=WavePlayer_2, properties {soundWaveAssetPtr=/Game/Sounds/Fort_Swamp_Sounds/MudBlobB001}}
 {id=WavePlayer_3, properties {soundWaveAssetPtr=/Game/Sounds/Fort_Swamp_Sounds/MudBlobB003}}
 {id=WavePlayer_4, properties {soundWaveAssetPtr=/Game/Sounds/Fort_Swamp_Sounds/MudBlobB005}}
 {id=WavePlayer_5, properties {soundWaveAssetPtr=&MudBlobB002}}
 {id=WavePlayer_6, properties {soundWaveAssetPtr=&MudBlobB004}}
 {id=WavePlayer_7, properties {soundWaveAssetPtr=&MudBlobB001}}
 {id=WavePlayer_8, properties {soundWaveAssetPtr=&MudBlobB003}}
 {id=WavePlayer_9, properties {soundWaveAssetPtr=&MudBlobB005}}
]
properties {
 bOverrideAttenuation=true
 bOverrideConcurrency=true
 maxDistance=1200
 soundClassObject=&Fort_Ambient_Props_Athena
 volumeMultiplier=1
 attenuationOverrides {
  bAttenuateWithLPF=true
  bEnableFocusInterpolation=true
  bEnableListenerFocus=true
  falloffDistance=1000
  focusAzimuth=20
  focusReleaseInterpSpeed=0.4
  hPFFrequencyAtMax=350
  lPFFrequencyAtMax=1500
  lPFRadiusMax=1000
  lPFRadiusMin=500
  nonFocusAzimuth=45
  nonFocusDistanceScale=1.5
  nonFocusVolumeAttenuation=0.1
  nonSpatializedRadiusStart=100
  stereoSpread=20
  attenuationShapeExtents {x=200}
 }
 busSends [{audioBus=null, minSendLevel=0.6, sendLevel=1, soundSourceBus=&AmbientAudio_Exterior_SourceBus}]
 concurrencyOverrides {maxCount=4, voiceStealReleaseTime=0.5}
}
